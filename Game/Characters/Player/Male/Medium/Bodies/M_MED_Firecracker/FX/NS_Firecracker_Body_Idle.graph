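# Niagara system for the Firecracker player body's idle FX: flames, glows, lights and sparks attached to bones or sockets on the character's skeletal mesh.
# Spawn groups pick each particle's attach point and index per-bone arrays for scale and alpha; Flame, Glow and Lights re-sample it every update, while Sparks fly outward.
# ADSCull takes MasterAlpha, and hidden hand/foot flags mask affected points.
# Sparks stop spawning while skydiving or driving.
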NiagaraSystem /Game/Characters/Player/Male/Medium/Bodies/M_MED_Firecracker/FX/NS_Firecracker_Body_Idle
dependencies {
 usedDataInterfaces [
  &NiagaraDataInterfaceArrayFloat
  &NiagaraDataInterfaceVectorCurve
  &NiagaraDataInterfaceCurve
  &NiagaraDataInterfaceColorCurve
 ]
 usedDynamicInputs []
 usedModules [
  &SystemState
  &self:SystemSpawnScript.SystemScriptSource.SystemScriptGraph.NiagaraNodeAssignment_0.SetVariables_#1
  &self:SystemSpawnScript.SystemScriptSource.SystemScriptGraph.NiagaraNodeAssignment_1.SetVariables_#2
  &self:SystemSpawnScript.SystemScriptSource.SystemScriptGraph.NiagaraNodeAssignment_2.SetVariables_#3
 ]
}
emitters {
 Flame {
  emitterUpdateScript [
   &EmitterState()
   &SpawnRate(Spawn Probability=HlslExpression() {hlslExpression="Engine.Emitter.NumParticles  >= 15 ? 0 : 1"}, SpawnRate=100, Use Spawn Probability=true)
   /SetVariables(Emitter.BoneFXArray=DataInterface() {propertyValues {GpuSyncMode=SyncCpuToGpu, MaxElements=0, FloatData [1,1,1,1,1,1,1,1]}}) {isSetParameters=true}
  ]
  particleSpawnScript [
   &FN_SetSpawnGroup(Number of Spawn Groups=8)
   &InitializeParticle(
    Color=<rgba 1 0.0833 0 1>
    Lifetime Max=0.2
    Lifetime Min=0.05
    Lifetime Mode=::NewEnumerator1
    Mesh Scale Max=<0.12 0.3 0.3>
    Mesh Scale Min=<0.06 0.05 0.05>
    Mesh Scale Mode=::NewEnumerator2
    Sprite Size Mode=::NewEnumerator3
    UsePositionOffset=false
    Write Material Random=true
   )
   &InitialMeshOrientation(
    Rotation=&RandomRangeVector(Minimum=<0 -0.05 -0.05>, Maximum=<1 0.05 0.05>)
    Use Orientation Vector=true
    Use Rotation Vector=true
   )
  ]
  particleUpdateScript [
   &self:ADSCull(
    BoneIndex:NiagaraInt32=$Particles.SpawnGroup
    FloatArray:NiagaraDataInterfaceArrayFloat=$Emitter.BoneFXArray
    MasterAlpha:NiagaraFloat=$User.MasterAlpha
   ) {module=ADSCull}
   &ParticleState()
   &SkeletalMeshLocation(
    Bone / Socket Index=$Particles.SpawnGroup
    MaskSpawnGroup=false
    Mesh Sampling Type=::NewEnumerator5
    Skeletal Mesh=$User.SkeletalMesh
    Socket Sampling Mode=::NewEnumerator5
   )
   &CameraOffset(Camera Offset Amount=2)
   &Color(
    Color=<rgba 1 0.0833 0 1>
    Scale Color=&VectorFromCurve(VectorCurve=DataInterface() {propertyValues {
       XCurve {keys [<at -0.0031 48.8>,<at 1 1>]}
       YCurve {keys [<at -0.006 20.3>,<at 1 0.8>]}
       ZCurve {keys [<at -0.0011 1.0067>,<at 0.9988 0.28>]}
      }})
   ) {enabled=false}
   &ScaleMeshSize(Scale Factor=&VectorFromCurve(
     VectorCurve=DataInterface() {propertyValues {
       XCurve {keys [<at 0 0.55>,<at 0.2 0.9>,<at 1 0>]}
       YCurve {keys [<at 0 0.05>,<at 0.2 0.95>,<at 1 0>]}
       ZCurve {keys [<at 0 0.55>,<at 0.2 0.9>,<at 1 0>]}
      }}
     Scale Curve=&VectorFromFloat(Value=&SelectFloatFromArray(Direct Array Index=$Particles.SpawnGroup, Float Selection Array=$System.SpawnGroupScalar))
    ))
   &ScaleColor(
    Scale Alpha=&MakeCustomFloatFromBool(
     Bool=$User.bisFootHidden
     True Float=&Multiply_Float(
      A=&RandomRangeFloat(Minimum=0.5, Maximum=1)
      B=&SelectFloatFromArray(Direct Array Index=$Particles.SpawnGroup, Float Selection Array=DataInterface() {propertyValues {FloatData [0,0,1,1,1,1,1,1]}})
     )
     False Float=&Multiply_Float(
      A=&RandomRangeFloat(Minimum=0.5, Maximum=1)
      B=&SelectFloatFromArray(Direct Array Index=$Particles.SpawnGroup, Float Selection Array=$Emitter.BoneFXArray)
     )
    )
    Scale RGB=&VectorFromFloat(Value=&FloatFromCurve(
      FloatCurve=DataInterface() {propertyValues {Curve {keys [
          {interpMode=RCIM_Cubic, leaveTangent=-2.1872, tangentMode=RCTM_Break, time=0, value=1}
          {arriveTangent=-1.0539, interpMode=RCIM_Cubic, tangentMode=RCTM_Break, time=1, value=0}
         ]}}}
      Scale Curve=&RandomRangeFloat(Minimum=2, Maximum=3.5)
     ))
    ScaleA=true
    ScaleRGB=true
   )
   &ScaleMeshSize(Scale Factor=&VectorFromFloat(Value=&MakeCustomFloatFromBool(
      Bool=$User.bisFootHidden
      True Float=&MakeCustomFloatFromBool(
       Bool=$User.bisHandHidden
       True Float=0
       False Float=&SelectFloatFromArray(Direct Array Index=&SpawnGroup(), Float Selection Array=$System.Hidden Foot Array)
      )
      False Float=&MakeCustomFloatFromBool(
       Bool=$User.bisHandHidden
       True Float=&SelectFloatFromArray(Direct Array Index=&SpawnGroup(), Float Selection Array=$System.Hidden Hand Array)
       False Float=1
      )
     ))) {module=ScaleMeshSize001}
   /SetVariables(Particles.VisibilityTag=&self:PartHiddenBoneCheck(
     is Foot Hidden:NiagaraBool=$User.bisFootHidden
     is Hand Hidden:NiagaraBool=$User.bisHandHidden
     Selection:NiagaraInt32=&SpawnGroup()
    )) {isSetParameters=true, module=SetVariables_#4}
  ]
  rendererClasses [&NiagaraMeshRendererProperties]
  renderers [{rendererClass=&NiagaraMeshRendererProperties, rendererIndex=0}]
 }
 Glow {
  emitterUpdateScript [
   &EmitterState(Life Cycle Mode=::NewEnumerator1, Loop Duration=0.15, UseLoopDelay=false)
   &SpawnBurst_Instantaneous(Spawn Count=8, Use Spawn Probability=false)
   /SetVariables() {isSetParameters=true, module=SetVariables_#5}
   /SetVariables(Emitter.BoneFXArray=DataInterface() {propertyValues {GpuSyncMode=SyncCpuToGpu, MaxElements=0, FloatData [1,1,1,1,1,1,1,1]}}) {isSetParameters=true, module=SetVariables_#6}
  ]
  particleSpawnScript [
   &FN_SetSpawnGroup(Number of Spawn Groups=8)
   &InitializeParticle(
    Color=<rgba 1.2 0.2583 0.1146 1>
    Lifetime=0.25
    Sprite Rotation Mode=::NewEnumerator0
    Sprite Size Mode=::NewEnumerator3
    Uniform Sprite Size=25
    UsePositionOffset=false
    Write Material Random=true
   )
   &DynamicMaterialParameters(
    Index 0 Param 1=1
    Index 1 Param 1=30
    Module.Param0WriteEnabled=true
    Module.Param0WriteEnabled1=true
    Module.Param1WriteEnabled=false
    Module.Param1WriteEnabled1=false
    Module.Param2WriteEnabled=false
    Module.Param2WriteEnabled1=false
    Module.Param3WriteEnabled=false
    Module.Param3WriteEnabled1=false
    Write Parameter Index 1=true
   )
  ]
  particleUpdateScript [
   &self:ADSCull(
    BoneIndex:NiagaraInt32=$Particles.SpawnGroup
    FloatArray:NiagaraDataInterfaceArrayFloat=$Emitter.BoneFXArray
    MasterAlpha:NiagaraFloat=$User.MasterAlpha
   ) {module=ADSCull}
   &ParticleState()
   &SkeletalMeshLocation(
    Bone / Socket Index=$Particles.SpawnGroup
    MaskSpawnGroup=false
    Mesh Sampling Type=::NewEnumerator5
    Sampled Position Offset=<2 0 0>
    Skeletal Mesh=$User.SkeletalMesh
    Socket Sampling Mode=::NewEnumerator5
   )
   &CameraOffset(Camera Offset Amount=5)
   &ScaleColor(
    Scale Alpha=&MakeCustomFloatFromBool(
     Bool=$User.bisFootHidden
     True Float=&Multiply_Float(
      A=&RandomRangeFloat(Minimum=0.5, Maximum=1)
      B=&SelectFloatFromArray(Direct Array Index=$Particles.SpawnGroup, Float Selection Array=DataInterface() {propertyValues {FloatData [0,0,1,1,1,1,1,1]}})
     )
     False Float=&Multiply_Float(
      A=&RandomRangeFloat(Minimum=0.5, Maximum=1)
      B=&SelectFloatFromArray(Direct Array Index=$Particles.SpawnGroup, Float Selection Array=$Emitter.BoneFXArray)
     )
    )
    Scale RGB=&VectorFromFloat(Value=1)
    ScaleA=true
    ScaleRGB=false
   )
   &ScaleSpriteSize(
    Scale Sprite Size Mode=::NewEnumerator0
    Uniform Scale Factor=&Waveform(
     "Waveform [1]"=::NewEnumerator0
     "[1] Phase Scale"=&RandomRangeFloat(Minimum=3, Maximum=6)
     "[1] Frequency"=1
     "[1] Amplitude Min/Max"=<0.8 1>
     "[2] Phase"=$Emitter.Age
     "[3] Phase"=$Emitter.Age
    )
   )
   &SpriteRotationRate(Rotation Rate=&RandomRangeFloat(Minimum=-10, Maximum=10))
   &ScaleSpriteSize(
    Scale Sprite Size Mode=::NewEnumerator0
    Uniform Scale Factor=&SelectFloatFromArray(Direct Array Index=$Particles.SpawnGroup, Float Selection Array=$System.SpawnGroupScalar)
   ) {module=ScaleSpriteSize001}
   &ScaleSpriteSize(
    Scale Sprite Size Mode=::NewEnumerator0
    Uniform Scale Factor=&MakeCustomFloatFromBool(
     Bool=$User.bisFootHidden
     True Float=&MakeCustomFloatFromBool(
      Bool=$User.bisHandHidden
      True Float=0
      False Float=&SelectFloatFromArray(Direct Array Index=&SpawnGroup(), Float Selection Array=$System.Hidden Foot Array)
     )
     False Float=&MakeCustomFloatFromBool(
      Bool=$User.bisHandHidden
      True Float=&SelectFloatFromArray(Direct Array Index=&SpawnGroup(), Float Selection Array=$System.Hidden Hand Array)
      False Float=1
     )
    )
   ) {module=ScaleSpriteSize002}
   /SetVariables(Particles.VisibilityTag=&self:PartHiddenBoneCheck(
     is Foot Hidden:NiagaraBool=$User.bisFootHidden
     is Hand Hidden:NiagaraBool=$User.bisHandHidden
     Selection:NiagaraInt32=&SpawnGroup()
    )) {isSetParameters=true, module=SetVariables_#4}
  ]
  rendererClasses [&NiagaraSpriteRendererProperties]
  renderers [{rendererClass=&NiagaraSpriteRendererProperties}]
 }
 Lights {
  emitterUpdateScript [
   &EmitterState(Life Cycle Mode=::NewEnumerator1, Loop Duration=0.25, UseLoopDelay=false)
   &SpawnBurst_Instantaneous(Spawn Count=4, Use Spawn Probability=false)
   /SetVariables() {isSetParameters=true, module=SetVariables_#5}
   /SetVariables(Emitter.BoneFXArray=DataInterface() {propertyValues {GpuSyncMode=SyncCpuToGpu, MaxElements=0, FloatData [1,1,1,1]}}) {isSetParameters=true, module=SetVariables_#6}
  ]
  particleSpawnScript [
   &FN_SetSpawnGroup()
   &InitializeParticle(
    Color=<rgba 6 1.2914 0.5729 1>
    Lifetime=0.25
    Sprite Rotation Mode=::NewEnumerator0
    Sprite Size Mode=::NewEnumerator3
    Uniform Sprite Size=25
    UsePositionOffset=false
    Write Material Random=true
   )
  ]
  particleUpdateScript [
   &self:ADSCull(
    BoneIndex:NiagaraInt32=$Particles.SpawnGroup
    FloatArray:NiagaraDataInterfaceArrayFloat=$Emitter.BoneFXArray
    MasterAlpha:NiagaraFloat=$User.MasterAlpha
   ) {module=ADSCull}
   &ParticleState()
   &SkeletalMeshLocation(
    Bone / Socket Index=$Particles.SpawnGroup
    MaskSpawnGroup=false
    Mesh Sampling Type=::NewEnumerator5
    Sampled Position Offset=<2 0 0>
    Skeletal Mesh=$User.SkeletalMesh
    Socket Sampling Mode=::NewEnumerator5
   )
   &CameraOffset(Camera Offset Amount=5)
   &FN_Light_Attributes(
    Color=<rgba 0.9635 0.29 0 1>
    Create Color Attribute=true
    Create Exponent Attribute=true
    Create Position Attribute=true
    Create Radius Attribute=true
    Position:NiagaraPosition=$Particles.SkeletalMeshLocation.SampledPosition
    Radius=&MakeCustomFloatFromBool(
     Bool=$User.bisFootHidden
     True Float=&Multiply_Float(
      A=&Multiply_Float(
       A=75
       B=&SelectFloatFromArray(Direct Array Index=$Particles.SpawnGroup, Float Selection Array=DataInterface() {propertyValues {FloatData [0,0,1,1,1,1,1,1]}})
      )
      B=&SelectFloatFromArray(Direct Array Index=$Particles.SpawnGroup, Float Selection Array=$System.SpawnGroupScalar)
     )
     False Float=&Multiply_Float(
      A=&Multiply_Float(
       A=75
       B=&SelectFloatFromArray(Direct Array Index=$Particles.SpawnGroup, Float Selection Array=$Emitter.BoneFXArray)
      )
      B=&SelectFloatFromArray(Direct Array Index=$Particles.SpawnGroup, Float Selection Array=$System.SpawnGroupScalar)
     )
    )
    Scale Color=&Multiply_Float(
     A=5
     B=&Waveform(
      "Waveform [1]"=::NewEnumerator0
      "[1] Phase Scale"=&RandomRangeFloat(Minimum=2, Maximum=6)
      "[1] Frequency"=1
      "[1] Amplitude Min/Max"=<0.6 1>
      "[2] Phase"=$Emitter.Age
      "[3] Phase"=$Emitter.Age
     )
    )
   )
   &ScaleSpriteSize(
    Scale Sprite Size Mode=::NewEnumerator0
    Uniform Scale Factor=&SelectFloatFromArray(Direct Array Index=$Particles.SpawnGroup, Float Selection Array=$System.SpawnGroupScalar)
   ) {module=ScaleSpriteSize001}
   /SetVariables(Particles.VisibilityTag=&self:PartHiddenBoneCheck(
     is Foot Hidden:NiagaraBool=$User.bisFootHidden
     is Hand Hidden:NiagaraBool=$User.bisHandHidden
     Selection:NiagaraInt32=&SpawnGroup()
    )) {isSetParameters=true, module=SetVariables_#4}
  ]
  rendererClasses [&NiagaraLightRendererProperties]
  renderers [{rendererClass=&NiagaraLightRendererProperties, rendererIndex=0}]
 }
 Sparks {
  emitterUpdateScript [
   &EmitterState()
   &SpawnRate(
    Spawn Probability=&Multiply_Float(
     A=0.4
     B=&MakeCustomFloatFromBool(Bool=&Boolean_OrOperation(Bool1=$User.IsSkydiving, Bool2=$User.IsDriving), True Float=0, False Float=1)
    )
    SpawnRate=75
    Use Spawn Probability=true
   )
   /SetVariables(Emitter.BoneFXArray=DataInterface() {propertyValues {GpuSyncMode=SyncCpuToGpu, MaxElements=0, FloatData [1,1,1,1,1,1,1,1]}}) {isSetParameters=true}
  ]
  particleSpawnScript [
   &FN_SetSpawnGroup(Number of Spawn Groups=8)
   &InitializeParticle(
    Color=<rgba 6 1.2914 0.5729 1>
    Lifetime Max=0.3
    Lifetime Min=0.1
    Lifetime Mode=::NewEnumerator1
    Sprite Rotation Mode=::NewEnumerator2
    Sprite Size Mode=::NewEnumerator1
    Uniform Sprite Size Max=2.5
    Uniform Sprite Size Min=1.5
    UsePositionOffset=false
    Write Material Random=true
   )
   &SkeletalMeshLocation(
    Bone / Socket Index=$Particles.SpawnGroup
    Interpolate Mesh Sampling=false
    MaskSpawnGroup=false
    Mesh Sampling Type=::NewEnumerator5
    Sampled Position Offset=<2 0 0>
    Skeletal Mesh=$User.SkeletalMesh
    Socket Sampling Mode=::NewEnumerator5
   )
   &SphereLocation(
    MaskSpawnGroup=false
    Offset Coordinate Space:ENiagaraCoordinateSpace=::Simulation
    Override Seed=false
    Sphere Coordinate Space=::Simulation
    Sphere Radius=2
    Surface Only=false
    UseNonUniformScale=false
    UseOffset=true
    UseSphereOrientation=false
   ) {module=SphereLocation001}
   &AddVelocityFromPoint(
    Module.Use Falloff Distance=false
    Origin Offset Coordinate Space=::Simulation
    Velocity Origin:NiagaraPosition=$Output.SkeletalMeshLocation.SampledPositionWithoutOffset
    Velocity Strength=&RandomRangeFloat(Minimum=50, Maximum=200)
   )
  ]
  particleUpdateScript [
   &self:ADSCull(
    BoneIndex:NiagaraInt32=$Particles.SpawnGroup
    FloatArray:NiagaraDataInterfaceArrayFloat=$Emitter.BoneFXArray
    MasterAlpha:NiagaraFloat=$User.MasterAlpha
   ) {module=ADSCull}
   &ParticleState()
   &Drag(Drag=10, Use Linear Drag=true, Use Rotational Drag=false)
   &SolveForcesAndVelocity(Clamp Velocity=false, Limit Acceleration=false)
   &ScaleColor(
    Scale Alpha=&MakeCustomFloatFromBool(
     Bool=$User.bisFootHidden
     True Float=&Multiply_Float(
      A=&RandomRangeFloat(Minimum=0.5, Maximum=1)
      B=&SelectFloatFromArray(Direct Array Index=$Particles.SpawnGroup, Float Selection Array=DataInterface() {propertyValues {FloatData [0,0,1,1,1,1,1,1]}})
     )
     False Float=&Multiply_Float(
      A=&RandomRangeFloat(Minimum=0.5, Maximum=1)
      B=&SelectFloatFromArray(Direct Array Index=$Particles.SpawnGroup, Float Selection Array=$Emitter.BoneFXArray)
     )
    )
    Scale RGB=&VectorFromFloat(Value=&FloatFromCurve(
      FloatCurve=DataInterface() {propertyValues {Curve {keys [
          {interpMode=RCIM_Cubic, leaveTangent=0.1747, tangentMode=RCTM_Break, time=0, value=1}
          {interpMode=RCIM_Cubic, tangentMode=RCTM_Break, time=1, value=0.2}
         ]}}}
      CurveIndex=&RandomRangeFloat(Minimum=0, Maximum=1)
      Scale Curve=10
     ))
    ScaleA=true
    ScaleRGB=true
   )
   &ScaleSpriteSize(
    Uniform Curve Scale=&MakeCustomFloatFromBool(
     Bool=$User.bisFootHidden
     True Float=&MakeCustomFloatFromBool(
      Bool=$User.bisHandHidden
      True Float=0
      False Float=&SelectFloatFromArray(Direct Array Index=&SpawnGroup(), Float Selection Array=$System.Hidden Foot Array)
     )
     False Float=&MakeCustomFloatFromBool(
      Bool=$User.bisHandHidden
      True Float=&SelectFloatFromArray(Direct Array Index=&SpawnGroup(), Float Selection Array=$System.Hidden Hand Array)
      False Float=1
     )
    )
    Uniform Curve Sprite Scale=DataInterface() {propertyValues {Curve {keys [<at 0 1 RCIM_Cubic>, {arriveTangent=-1.8095, interpMode=RCIM_Cubic, tangentMode=RCTM_Break, time=1, value=0.1}]}}}
   )
   &SpriteRotationRate(Rotation Rate=&RandomRangeFloat(Minimum=-250, Maximum=250))
   /SetVariables(Particles.VisibilityTag=&self:PartHiddenBoneCheck(
     is Foot Hidden:NiagaraBool=$User.bisFootHidden
     is Hand Hidden:NiagaraBool=$User.bisHandHidden
     Selection:NiagaraInt32=&SpawnGroup()
    )) {isSetParameters=true, module=SetVariables_#4}
  ]
  rendererClasses [&NiagaraSpriteRendererProperties]
  renderers [{rendererClass=&NiagaraSpriteRendererProperties}]
 }
}
summary {
 systemName=NS_Firecracker_Body_Idle
 emitters [
  {emitterName=Flame, rendererClasses [&NiagaraMeshRendererProperties]}
  {emitterName=Sparks, rendererClasses [&NiagaraSpriteRendererProperties]}
  {emitterName=Glow, rendererClasses [&NiagaraSpriteRendererProperties]}
  {emitterName=Lights, rendererClasses [&NiagaraLightRendererProperties]}
 ]
 userVariables [
  User.bisFootHidden {type=NiagaraBool}
  User.bisHandHidden {type=NiagaraBool}
  User.FrontendSpawn {type=NiagaraBool}
  User.IsDriving {type=NiagaraBool}
  User.IsSkydiving {type=NiagaraBool}
  User.MasterAlpha {type=NiagaraFloat, defaultValue=1, description=""}
  User.SkeletalMesh {
   type=NiagaraDataInterfaceSkeletalMesh
   defaultValue {
    struct=NiagaraExt_VariableValue_DataInterface
    value {dataInterface=&self:NiagaraDataInterfaceSkeletalMesh_0}
   }
  }
 ]
}
