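# Physics asset for the male medium Firecracker player body, covering the torso and two dynamic bandolier bone chains rather than a full ragdoll.
# Spheres from pelvis to spine_04 form the torso, joined by free angular constraints.
# Left and right capsule chains each run from a master bone on spine_04 through five segments with tight swing and twist limits.
# The master attachments differ: the right is free, while the left uses the chain limits.
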PhysicsAsset /Game/Characters/Player/Male/Medium/Bodies/M_MED_Firecracker/Meshes/M_MED_Firecracker_Physics
bodies {
 dyn_l_bandolier_01 [{shapeType=Capsule, length=2.5821, radius=6.2062, center=<3.2746 -3.9045 -3.9027>, rotation=<rot 39.9996 0 -10.0003>}]
 dyn_l_bandolier_02 [{shapeType=Capsule, length=5.7628, radius=8.1561, center=<6.0809 -5.1767 0.2264>, rotation=<rot 16.5212 -37.4133 -51.7953>}]
 dyn_l_bandolier_03 [{shapeType=Capsule, length=5.4668, radius=8.3293, center=<7.4982 0.0001 0>, rotation=<rot 0 0 89.9986>}]
 dyn_l_bandolier_04 [{shapeType=Capsule, length=7.3187, radius=6.0537, center=<1.9043 5.6284 -1.3334>, rotation=<rot 54.9992 0 30.0001>}]
 dyn_l_bandolier_05 [{shapeType=Capsule, length=0.505, radius=0.505}]
 dyn_l_bandolier_MSTR [{shapeType=Capsule, length=0.505, radius=0.505}]
 dyn_r_bandolier_01 [{shapeType=Capsule, length=3.7084, radius=5.0601, center=<6.0442 3.4755 -0.4021>, rotation=<rot 90 0 -9.9993>}]
 dyn_r_bandolier_02 [{shapeType=Capsule, length=6.6175, radius=6.0881, center=<1.7405 6.1271 -0.911>, rotation=<rot -84.9839 179.9999 -174.9987>}]
 dyn_r_bandolier_03 [{shapeType=Capsule, length=6.4239, radius=5.9556, center=<-1.7894 4.487 0.0269>, rotation=<rot 65 0 -20.0002>}]
 dyn_r_bandolier_04 [{shapeType=Capsule, length=7.4008, radius=5.7899, center=<1.0142 4.7467 0.2634>, rotation=<rot 90 0 0>}]
 dyn_r_bandolier_05 [{shapeType=Capsule, length=0.5052, radius=3.0993, center=<0 1.605 0>}]
 dyn_r_bandolier_MSTR [{shapeType=Capsule, length=0.505, radius=0.505}]
 pelvis [{shapeType=Sphere, radius=16.455, center=<0 0.0356 0>}]
 spine_01 [{shapeType=Sphere, radius=16.1728, center=<0 0.436 0>}]
 spine_02 [{shapeType=Sphere, radius=16.1978, center=<0 0.4962 0>}]
 spine_03 [{shapeType=Sphere, radius=16.1631, center=<0 0.7151 0>}]
 spine_04 [{shapeType=Sphere, radius=16.203, center=<0 1.3794 0>}]
}
constraints [
 {
  bone1Name=spine_01
  bone2Name=pelvis
  swing1LimitDegrees=45
  swing1Motion=Free
  swing2LimitDegrees=45
  swing2Motion=Free
  twistLimitDegrees=45
  twistMotion=Free
 }
 {
  bone1Name=spine_02
  bone2Name=spine_01
  swing1LimitDegrees=45
  swing1Motion=Free
  swing2LimitDegrees=45
  swing2Motion=Free
  twistLimitDegrees=45
  twistMotion=Free
 }
 {
  bone1Name=spine_03
  bone2Name=spine_02
  swing1LimitDegrees=45
  swing1Motion=Free
  swing2LimitDegrees=45
  swing2Motion=Free
  twistLimitDegrees=45
  twistMotion=Free
 }
 {
  bone1Name=spine_04
  bone2Name=spine_03
  swing1LimitDegrees=45
  swing1Motion=Free
  swing2LimitDegrees=45
  swing2Motion=Free
  twistLimitDegrees=45
  twistMotion=Free
 }
 {
  bone1Name=dyn_r_bandolier_MSTR
  bone2Name=spine_04
  swing1LimitDegrees=45
  swing1Motion=Free
  swing2LimitDegrees=45
  swing2Motion=Free
  twistLimitDegrees=45
  twistMotion=Free
 }
 {bone1Name=dyn_r_bandolier_01, bone2Name=dyn_r_bandolier_MSTR, swing1LimitDegrees=20, swing2LimitDegrees=10, twistLimitDegrees=3}
 {bone1Name=dyn_r_bandolier_02, bone2Name=dyn_r_bandolier_01, swing1LimitDegrees=20, swing2LimitDegrees=10, twistLimitDegrees=3}
 {bone1Name=dyn_r_bandolier_03, bone2Name=dyn_r_bandolier_02, swing1LimitDegrees=20, swing2LimitDegrees=10, twistLimitDegrees=3}
 {bone1Name=dyn_r_bandolier_04, bone2Name=dyn_r_bandolier_03, swing1LimitDegrees=20, swing2LimitDegrees=10, twistLimitDegrees=3}
 {bone1Name=dyn_r_bandolier_05, bone2Name=dyn_r_bandolier_04, swing1LimitDegrees=20, swing2LimitDegrees=10, twistLimitDegrees=3}
 {bone1Name=dyn_l_bandolier_05, bone2Name=dyn_l_bandolier_04, swing1LimitDegrees=20, swing2LimitDegrees=10, twistLimitDegrees=3}
 {bone1Name=dyn_l_bandolier_04, bone2Name=dyn_l_bandolier_03, swing1LimitDegrees=20, swing2LimitDegrees=10, twistLimitDegrees=3}
 {bone1Name=dyn_l_bandolier_03, bone2Name=dyn_l_bandolier_02, swing1LimitDegrees=20, swing2LimitDegrees=10, twistLimitDegrees=3}
 {bone1Name=dyn_l_bandolier_02, bone2Name=dyn_l_bandolier_01, swing1LimitDegrees=20, swing2LimitDegrees=10, twistLimitDegrees=3}
 {bone1Name=dyn_l_bandolier_01, bone2Name=dyn_l_bandolier_MSTR, swing1LimitDegrees=20, swing2LimitDegrees=10, twistLimitDegrees=3}
 {bone1Name=dyn_l_bandolier_MSTR, bone2Name=spine_04, swing1LimitDegrees=20, swing2LimitDegrees=10, twistLimitDegrees=3}
]
