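# Post-processing AnimBlueprint for the Firecracker medium male player body, adding secondary motion to bandoliers, bombs and fuses.
# It copies the parent mesh pose, applies two bool-gated Control Rigs (general and rifle), then offsets shoulder bones while parachuting.
# Weighted multi-bone constraints anchor bandoliers and bombs, interleaved with RigidBody and AnimDynamics simulation.
# The event graph sets both rig flags from weapon type, relaxed, skydiving and targeting states.
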AnimBlueprint /Game/Characters/Player/Male/Medium/Bodies/M_MED_Firecracker/Meshes/M_MED_Firecracker_AnimBP
target_skeleton=&/M_MED_Firecracker_Skeleton
graphs [
 AnimGraph {nodes [
   Constraint!(
    ComponentPose
    Alpha=1
    ConstraintWeights_0=0.0415
    ConstraintWeights_1:float=0
    ConstraintWeights_2:float=0.072
    ConstraintWeights_3:float=0.1826
    ConstraintWeights_4:float=0.1706
    ConstraintWeights_5:float=0.0998
    ConstraintWeights_6:float=0
    ConstraintWeights_7:float=0
    ConstraintWeights_8:float=0.3956
   ) -> (Pose) {
    alphaBoolBlend {blendOption=Linear}
    constraintSetup [
     {targetBone {boneName=upperarm_twist_01_r}}
     {transformType=Translation, targetBone {boneName=None}}
     {targetBone {boneName=spine_03}}
     {targetBone {boneName=spine_04}}
     {targetBone {boneName=spine_05}}
     {targetBone {boneName=clavicle_r}}
     {transformType=Translation, targetBone {boneName=None}}
     {transformType=Translation, targetBone {boneName=None}}
     {targetBone {boneName=dyn_r_shoulder}}
    ]
    constraintWeights [1,1,1,1,1,1,1,1,1]
    boneToModify=~
   }
   n1 = Control Rig!(Source=Copy Pose From Mesh() {bUseAttachedParent=true}, bAlphaBoolEnabled:bool=Get ctrl rig() -> (ctrl rig:bool)) -> (Pose) {
    alphaInputType=Bool
    alphaBoolBlend {blendInTime=0.1, blendOption=ExpOut, blendOutTime=1}
    controlRigAssetReference {blueprintRigClass=&/M_MED_Firecracker_CtrlRig.M_MED_Firecracker_CtrlRig_C}
   }
   n2 = Control Rig(Source=@n1, bAlphaBoolEnabled=Get Rifle() -> (Rifle:bool)) {
    alphaInputType=Bool
    alphaBoolBlend {blendInTime=0.3, blendOption=Linear, blendOutTime=0.3}
    controlRigAssetReference {blueprintRigClass=&/M_MED_Firecracker_Rifle_CtrlRig.M_MED_Firecracker_Rifle_CtrlRig_C}
   }
   n3 = Get bIsParachuteOpened()
   n4 = "Transform (Modify) Bone - Bone: dyn_r_shoulder"!(
    ComponentPose=Local To Component(LocalPose=@n2)
    bAlphaBoolEnabled:bool=@n3
    Translation="1.582,0.000,8.159"
    Rotation="40.000,0.000,0.000"
    Scale="1,1,1"
   ) -> (Pose) {
    alphaInputType=Bool
    rotationMode=BMM_Additive
    translationMode=BMM_Additive
    alphaBoolBlend {blendInTime=0.2, blendOption=Linear, blendOutTime=0.1}
    boneToModify {boneName=dyn_r_shoulder}
    rotation {pitch=40}
    translation {x=1.582, z=8.159}
   }
   n5 = "Transform (Modify) Bone - Bone: dyn_l_shoulder"!(ComponentPose=@n4, bAlphaBoolEnabled:bool=@n3, Translation="2.873,0.000,5.824", Rotation="-25.000,0.000,0.000", Scale="1,1,1") -> (Pose) {
    alphaInputType=Bool
    rotationMode=BMM_Additive
    translationMode=BMM_Additive
    alphaBoolBlend {blendInTime=0.2, blendOption=Linear, blendOutTime=0.1}
    boneToModify {boneName=dyn_l_shoulder}
    rotation {pitch=-25}
    translation {x=2.873, z=5.824}
   }
   n6 = "Constraint - dyn_l_bandolier_MSTR (PPPPPPPP)"!(
    ComponentPose=@n5
    Alpha=1
    ConstraintWeights_0=0.0074
    ConstraintWeights_1:float=0.0333
    ConstraintWeights_2:float=0.0479
    ConstraintWeights_3:float=0.1696
    ConstraintWeights_4:float=0.2083
    ConstraintWeights_5:float=0.4607
    ConstraintWeights_6:float=0.0164
    ConstraintWeights_7:float=0.0564
   ) -> (Pose) {
    alphaBoolBlend {blendOption=Linear}
    boneToModify {boneName=dyn_l_bandolier_MSTR}
    constraintSetup [
     {targetBone {boneName=clavicle_l}}
     {targetBone {boneName=spine_01}}
     {targetBone {boneName=spine_02}}
     {targetBone {boneName=spine_03}}
     {targetBone {boneName=spine_04}}
     {targetBone {boneName=spine_05}}
     {targetBone {boneName=clavicle_r}}
     {targetBone {boneName=pelvis}}
    ]
    constraintWeights [1,1,1,1,1,1,1,1]
   }
   n7 = "Constraint - dyn_r_bandolier_MSTR (PPPPPPPP)"!(
    ComponentPose=@n6
    Alpha=1
    ConstraintWeights_0=0.0069
    ConstraintWeights_1:float=0.0277
    ConstraintWeights_2:float=0.0433
    ConstraintWeights_3:float=0.1677
    ConstraintWeights_4:float=0.2092
    ConstraintWeights_5:float=0.482
    ConstraintWeights_6:float=0.0193
    ConstraintWeights_7:float=0.044
   ) -> (Pose) {
    alphaBoolBlend {blendOption=Linear}
    boneToModify {boneName=dyn_r_bandolier_MSTR}
    constraintSetup [
     {targetBone {boneName=clavicle_l}}
     {targetBone {boneName=spine_01}}
     {targetBone {boneName=spine_02}}
     {targetBone {boneName=spine_03}}
     {targetBone {boneName=spine_04}}
     {targetBone {boneName=spine_05}}
     {targetBone {boneName=clavicle_r}}
     {targetBone {boneName=pelvis}}
    ]
    constraintWeights [1,1,1,1,1,1,1,1]
   }
   n8 = "Constraint - dyn_l_bandolier_06 (PPPPPPPP)"!(
    ComponentPose=@n7
    Alpha=1
    ConstraintWeights_0=0.0039
    ConstraintWeights_1:float=0.0997
    ConstraintWeights_2:float=0.1216
    ConstraintWeights_3:float=0.1777
    ConstraintWeights_4:float=0.3108
    ConstraintWeights_5:float=0.1763
    ConstraintWeights_6:float=0.1078
    ConstraintWeights_7:float=0.0022
   ) -> (Pose) {
    alphaBoolBlend {blendOption=Linear}
    boneToModify {boneName=dyn_l_bandolier_06}
    constraintSetup [
     {targetBone {boneName=clavicle_l}}
     {targetBone {boneName=spine_01}}
     {targetBone {boneName=spine_02}}
     {targetBone {boneName=spine_03}}
     {targetBone {boneName=spine_04}}
     {targetBone {boneName=spine_05}}
     {targetBone {boneName=pelvis}}
     {targetBone {boneName=neck_01}}
    ]
    constraintWeights [1,1,1,1,1,1,1,1]
   }
   n9 = "Constraint - dyn_r_bandolier_06 (PPPPPPPP)"!(
    ComponentPose=@n8
    Alpha=1
    ConstraintWeights_0=0.0012
    ConstraintWeights_1:float=0.108
    ConstraintWeights_2:float=0.1336
    ConstraintWeights_3:float=0.1863
    ConstraintWeights_4:float=0.3044
    ConstraintWeights_5:float=0.1446
    ConstraintWeights_6:float=0.0012
    ConstraintWeights_7:float=0.1207
   ) -> (Pose) {
    alphaBoolBlend {blendOption=Linear}
    boneToModify {boneName=dyn_r_bandolier_06}
    constraintSetup [
     {targetBone {boneName=clavicle_l}}
     {targetBone {boneName=spine_01}}
     {targetBone {boneName=spine_02}}
     {targetBone {boneName=spine_03}}
     {targetBone {boneName=spine_04}}
     {targetBone {boneName=spine_05}}
     {targetBone {boneName=clavicle_r}}
     {targetBone {boneName=pelvis}}
    ]
    constraintWeights [1,1,1,1,1,1,1,1]
   }
   n10 = RigidBody(ComponentPose=@n9) {
    bClampLinearTranslationLimitToRefPose=true
    bOverrideWorldGravity=true
    overridePhysicsAsset=&/M_MED_Firecracker_Physics
    simulationSpace=BaseBoneSpace
    alphaBoolBlend {blendInTime=0.1, blendOption=Linear, blendOutTime=0.1}
    baseBoneRef {boneName=spine_03}
    componentAppliedLinearAccClamp=<500 500 500>
    componentLinearAccScale=<0.3 0.3 0.3>
    componentLinearVelScale=<0.2 0.2 0.25>
    overrideWorldGravity {z=-980}
   }
   n11 = "Constraint - dyn_r_bandolier_05 (P)"(ComponentPose=@n10, Alpha=1, ConstraintWeights_0=1) {
    alphaBoolBlend {blendOption=Linear}
    boneToModify {boneName=dyn_r_bandolier_05}
    constraintSetup [{targetBone {boneName=dyn_r_bandolier_06}}]
    constraintWeights [1]
   }
   n12 = "Constraint - dyn_l_bandolier_05 (P)"(ComponentPose=@n11, Alpha=1, ConstraintWeights_0=1) {
    alphaBoolBlend {blendOption=Linear}
    boneToModify {boneName=dyn_l_bandolier_05}
    constraintSetup [{targetBone {boneName=dyn_l_bandolier_06}}]
    constraintWeights [1]
   }
   n13 = "Transform (Modify) Bone - Bone: dyn_l_bandolier_03"!(ComponentPose=@n12, bAlphaBoolEnabled:bool=True, Translation="-0.000,0.000,4", Rotation="0,0,0", Scale="1,1,1") -> (Pose) {
    alphaInputType=Bool
    translationMode=BMM_Additive
    alphaBoolBlend {blendInTime=0.1, blendOption=Linear, blendOutTime=0.1}
    boneToModify {boneName=dyn_l_bandolier_03}
    translation {z=4}
   }
   n14 = "Constraint - dyn_bomb_01_mstr (PPPPPPPP)"!(
    ComponentPose=@n13
    Alpha=1
    ConstraintWeights_0=0.005
    ConstraintWeights_1:float=0.1081
    ConstraintWeights_2:float=0.1201
    ConstraintWeights_3:float=0.1721
    ConstraintWeights_4:float=0.2979
    ConstraintWeights_5:float=0.1724
    ConstraintWeights_6:float=0.1213
    ConstraintWeights_7:float=0.0031
   ) -> (Pose) {
    alphaBoolBlend {blendOption=Linear}
    boneToModify {boneName=dyn_bomb_01_mstr}
    constraintSetup [
     {targetBone {boneName=clavicle_l}}
     {targetBone {boneName=spine_01}}
     {targetBone {boneName=spine_02}}
     {targetBone {boneName=spine_03}}
     {targetBone {boneName=spine_04}}
     {targetBone {boneName=spine_05}}
     {targetBone {boneName=pelvis}}
     {targetBone {boneName=neck_01}}
    ]
    constraintWeights [1,1,1,1,1,1,1,1]
   }
   n15 = "Constraint - dyn_bomb_02_mstr (PPPPPPPP)"!(
    ComponentPose=@n14
    Alpha=1
    ConstraintWeights_0=0.0267
    ConstraintWeights_1:float=0.0299
    ConstraintWeights_2:float=0.0431
    ConstraintWeights_3:float=0.075
    ConstraintWeights_4:float=0.0438
    ConstraintWeights_5:float=0.0299
    ConstraintWeights_6:float=0.5011
    ConstraintWeights_7:float=0.2505
   ) -> (Pose) {
    alphaBoolBlend {blendOption=Linear}
    boneToModify {boneName=dyn_bomb_02_mstr}
    constraintSetup [
     {targetBone {boneName=spine_01}}
     {targetBone {boneName=spine_02}}
     {targetBone {boneName=spine_03}}
     {targetBone {boneName=spine_04}}
     {targetBone {boneName=spine_05}}
     {targetBone {boneName=pelvis}}
     {targetBone {boneName=dyn_l_bandolier_04}}
     {targetBone {boneName=dyn_l_bandolier_05}}
    ]
    constraintWeights [1,1,1,1,1,1,1,1]
   }
   n16 = Anim Dynamics - Bone: dyn_bomb_01(ComponentPose=@n15, Alpha=1, ExternalForce="0,0,0") {
    angularDampingOverride=0.998
    angularSpringConstant=150
    bAngularSpring=true
    bGravityOverrideInSimSpace=true
    bOverrideAngularDamping=true
    bUseGravityOverride=true
    simulationSpace=World
    alphaBoolBlend {blendOption=Linear}
    boundBone {boneName=dyn_bomb_01}
    componentAppliedLinearAccClamp=<3000 3000 3000>
    componentLinearAccScale=<0.3 0.3 0.3>
    componentLinearVelScale=<0.3 0.3 0.3>
    gravityOverride {z=-980}
    physicsBodyDefinitions [{
      sphereCollisionRadius=0
      boundBone {boneName=dyn_bomb_01}
      boxExtents=<20 20 20>
      constraintSetup {angularLimitsMax=<1 15 25>, angularLimitsMin=<-1 -15 -5>, angularTarget=<0 0 0>}
      localJointOffset=<2 -0.5 -1>
     }]
    relativeSpaceBone {boneName=spine_02}
   }
   n17 = Anim Dynamics - Bone: dyn_bomb_02(ComponentPose=@n16, Alpha=1, ExternalForce="0,0,0") {
    angularDampingOverride=0.998
    angularSpringConstant=150
    bAngularSpring=true
    bGravityOverrideInSimSpace=true
    bOverrideAngularDamping=true
    bUseGravityOverride=true
    simulationSpace=World
    alphaBoolBlend {blendOption=Linear}
    boundBone {boneName=dyn_bomb_02}
    componentAppliedLinearAccClamp=<3000 3000 3000>
    componentLinearAccScale=<0.3 0.3 0.3>
    componentLinearVelScale=<0.3 0.3 0.3>
    gravityOverride {z=-980}
    physicsBodyDefinitions [{
      sphereCollisionRadius=0
      boundBone {boneName=dyn_bomb_02}
      boxExtents=<18 18 18>
      constraintSetup {angularLimitsMax=<1 15 5>, angularLimitsMin=<-1 -15 -25>, angularTarget=<0 0 0>}
      localJointOffset=<3 -0.5 -1>
     }]
    relativeSpaceBone {boneName=spine_02}
   }
   n18 = Anim Dynamics - Bone: dyn_bomb_01_fuse_01(ComponentPose=@n17, Alpha=1, ExternalForce="0,0,0") {
    angularDampingOverride=0.98
    angularSpringConstant=150
    bAngularSpring=true
    bGravityOverrideInSimSpace=true
    bOverrideAngularDamping=true
    simulationSpace=World
    alphaBoolBlend {blendOption=Linear}
    boundBone {boneName=dyn_bomb_01_fuse_01}
    componentLinearAccScale=<1.25 1.25 1.25>
    componentLinearVelScale=<1.25 1.25 1.25>
    physicsBodyDefinitions [{
      sphereCollisionRadius=0
      boundBone {boneName=dyn_bomb_01_fuse_01}
      boxExtents=<10 5 5>
      constraintSetup {angularLimitsMax=<2 25 25>, angularLimitsMin=<-2 -25 -25>, angularTarget=<0 0 0>}
      localJointOffset=<1.5 0 0>
     }]
    relativeSpaceBone {boneName=dyn_bomb_01_fuse_MSTR}
   }
   n19 = Anim Dynamics - Bone: dyn_bomb_01_fuse_02(ComponentPose=@n18, Alpha=1, ExternalForce="0,0,0") {
    angularDampingOverride=0.98
    angularSpringConstant=150
    bAngularSpring=true
    bGravityOverrideInSimSpace=true
    bOverrideAngularDamping=true
    simulationSpace=World
    alphaBoolBlend {blendOption=Linear}
    boundBone {boneName=dyn_bomb_01_fuse_02}
    componentLinearAccScale=<1.25 1.25 1.25>
    componentLinearVelScale=<1.25 1.25 1.25>
    physicsBodyDefinitions [{
      sphereCollisionRadius=0
      boundBone {boneName=dyn_bomb_01_fuse_02}
      boxExtents=<10 5 5>
      constraintSetup {angularLimitsMax=<2 25 25>, angularLimitsMin=<-2 -25 -25>, angularTarget=<0 0 0>}
      localJointOffset=<1.5 0 0>
     }]
    relativeSpaceBone {boneName=dyn_bomb_01_fuse_MSTR}
   }
   n20 = Anim Dynamics - Bone: dyn_bomb_02_fuse_01(ComponentPose=@n19, Alpha=1, ExternalForce="0,0,0") {
    angularDampingOverride=0.98
    angularSpringConstant=150
    bAngularSpring=true
    bGravityOverrideInSimSpace=true
    bOverrideAngularDamping=true
    simulationSpace=World
    alphaBoolBlend {blendOption=Linear}
    boundBone {boneName=dyn_bomb_02_fuse_01}
    componentLinearAccScale=<1.25 1.25 1.25>
    componentLinearVelScale=<1.25 1.25 1.25>
    physicsBodyDefinitions [{
      sphereCollisionRadius=0
      boundBone {boneName=dyn_bomb_02_fuse_01}
      boxExtents=<10 5 5>
      constraintSetup {angularLimitsMax=<2 25 25>, angularLimitsMin=<-2 -25 -25>, angularTarget=<0 0 0>}
      localJointOffset=<1.5 0 0>
     }]
    relativeSpaceBone {boneName=dyn_bomb_02_fuse_MSTR}
   }
   n21 = Anim Dynamics - Bone: dyn_bomb_02_fuse_02(ComponentPose=@n20, Alpha=1, ExternalForce="0,0,0") {
    angularDampingOverride=0.98
    angularSpringConstant=150
    bAngularSpring=true
    bGravityOverrideInSimSpace=true
    bOverrideAngularDamping=true
    simulationSpace=World
    alphaBoolBlend {blendOption=Linear}
    boundBone {boneName=dyn_bomb_02_fuse_02}
    componentLinearAccScale=<1.25 1.25 1.25>
    componentLinearVelScale=<1.25 1.25 1.25>
    physicsBodyDefinitions [{
      sphereCollisionRadius=0
      boundBone {boneName=dyn_bomb_02_fuse_02}
      boxExtents=<10 5 5>
      constraintSetup {angularLimitsMax=<2 25 25>, angularLimitsMin=<-2 -25 -25>, angularTarget=<0 0 0>}
      localJointOffset=<1.5 0 0>
     }]
    relativeSpaceBone {boneName=dyn_bomb_02_fuse_MSTR}
   }
   n22 = Anim Dynamics - Bone: dyn_ball_r(ComponentPose=@n21, Alpha=1, ExternalForce="0,0,0") {
    angularDampingOverride=0.998
    angularSpringConstant=150
    bAngularSpring=true
    bGravityOverrideInSimSpace=true
    bOverrideAngularDamping=true
    bUseGravityOverride=true
    simulationSpace=RootRelative
    alphaBoolBlend {blendOption=Linear}
    boundBone {boneName=dyn_ball_r}
    componentAppliedLinearAccClamp=<3000 3000 3000>
    componentLinearAccScale=<0.3 0.3 0.3>
    componentLinearVelScale=<0.3 0.3 0.3>
    gravityOverride {z=-980}
    physicsBodyDefinitions [{
      sphereCollisionRadius=0
      boundBone {boneName=dyn_ball_r}
      boxExtents=<20 10 10>
      constraintSetup {angularLimitsMax=<1 15 5>, angularLimitsMin=<-1 -25 -5>, angularTarget=<0 0 0>}
      localJointOffset=<2 0 0>
     }]
    relativeSpaceBone {boneName=root}
   }
   n23 = Anim Dynamics - Bone: dyn_ball_l(ComponentPose=@n22, Alpha=1, ExternalForce="0,0,0") {
    angularDampingOverride=0.998
    angularSpringConstant=150
    bAngularSpring=true
    bGravityOverrideInSimSpace=true
    bOverrideAngularDamping=true
    bUseGravityOverride=true
    simulationSpace=RootRelative
    alphaBoolBlend {blendOption=Linear}
    boundBone {boneName=dyn_ball_l}
    componentAppliedLinearAccClamp=<3000 3000 3000>
    componentLinearAccScale=<0.3 0.3 0.3>
    componentLinearVelScale=<0.3 0.3 0.3>
    gravityOverride {z=-980}
    physicsBodyDefinitions [{
      sphereCollisionRadius=0
      boundBone {boneName=dyn_ball_l}
      boxExtents=<20 10 10>
      constraintSetup {angularLimitsMax=<1 15 5>, angularLimitsMin=<-1 -25 -5>, angularTarget=<0 0 0>}
      localJointOffset=<2 0 0>
     }]
    relativeSpaceBone {boneName=root}
   }
   Output Pose(Result=Component To Local(ComponentPose=@n23)) {layerGroup=None}
   Get bIsSkyDiving()
  ]}
 EventGraph {nodes [
   Event BlueprintUpdateAnimation()
   > n1 = Sequence!(execute) -> (then_0, then_1, then_2:Exec, then_3:Exec, then_4:Exec, then_5:Exec)
   > n2 = Cast To FortPlayerAnimInstance(Object=GetAnimInstance(self=GetOwnerSkeletalMeshComponent()))
   > Set weapon(weapon:String=Enum to String!(Enumerator:EFortWeaponCoreAnimation Enum=Get WeaponCoreAnim(self=@n2)) -> (ReturnValue:String)) -> (Output_Get:String)
   n3 = Branch(
    execute=@n1.then_1
    Condition=OR Boolean!(A=Get bIsRelaxedLevel1() -> (bIsRelaxedLevel1:bool), B=Get bIsRelaxedLevel2() -> (bIsRelaxedLevel2:bool), C:bool) -> (ReturnValue)
   )
   > Set Rifle(Rifle:bool=false) -> (Output_Get:bool)
   n4 = Get weapon() -> (weapon:String)
   n5 = Get bIsSkyDiving()
   n6 = Branch(execute=@n3.else, Condition=AND Boolean(A="Equal Exactly (String)"(A=@n4, B=Assault Rifle), B=NOT Boolean(A=@n5)))
   > Set Rifle(Rifle=true)
   Set Rifle(execute=@n6.else, Rifle=false)
   n7 = Get weapon()
   Branch(execute=@n1.then_2, Condition=AND Boolean(A="Equal Exactly (String)"(A=@n7, B=Assault Rifle), B=Get bIsTargeting()))
   > Set Rifle(Rifle=true)
   n8 = Get bIsSkyDiving()
   n9 = Get bIsRelaxedLevel1()
   n10 = Get bIsRelaxedLevel2()
   n11 = Branch(execute=@n1.then_3, Condition=OR Boolean(A=@n8, B=@n9, C=@n10))
   > Set ctrl rig(ctrl rig:bool=false) -> (Output_Get:bool)
   Set ctrl rig(execute=@n11.else, ctrl rig=true)
   Branch(execute=@n1.then_5, Condition=Get bIsTargeting())
   > Set ctrl rig(ctrl rig=true)
   n12 = Branch()
   > Set ctrl rig(ctrl rig=true)
   Set ctrl rig(execute=@n12.else, ctrl rig=false)
   NOT Boolean(A=@n10)
   NOT Boolean(A=@n8)
   NOT Boolean(A=@n9)
   AND Boolean(A=false, B=false)
   Set Rifle(Rifle=false)
   Set ctrl rig(ctrl rig=false)
  ]}
]
variables [
 ctrl rig {type=boolean}
 Rifle {type=boolean}
 weapon {type=string, default=""}
]
